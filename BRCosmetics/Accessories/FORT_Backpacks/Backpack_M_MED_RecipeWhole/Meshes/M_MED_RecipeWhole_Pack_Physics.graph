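# Physics asset for the RecipeWhole backpack cosmetic mesh, giving its dangling arms, legs and bows simulated secondary motion.
# A base body of stacked capsules anchors six dynamic appendage bodies, each built from one or two capsules.
# Each appendage constrains directly to the base body with near-locked twist and per-type swing limits, widest on the arms and tightest on the bows.
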PhysicsAsset /BRCosmetics/Accessories/FORT_Backpacks/Backpack_M_MED_RecipeWhole/Meshes/M_MED_RecipeWhole_Pack_Physics
bodies {
 base_body [
  {shapeType=Capsule, length=0, radius=6.6571, rotation=<rot 90 0 0>}
  {shapeType=Capsule, length=1.5, radius=7.25, shapeName=base_body_capsule, center=<0 0 -5.7535>, rotation=<rot 90 0 0>}
  {shapeType=Capsule, length=7.728, radius=11, shapeName=base_body_capsule, center=<0 0 13.0824>, rotation=<rot 90 0 0>}
 ]
 dyn_arm_l [
  {shapeType=Capsule, length=3.6134, radius=2, shapeName=dyn_arm_l_capsule, center=<4 0 0>, rotation=<rot 90 0 0>}
  {shapeType=Capsule, length=3.1798, radius=1.4409, shapeName=dyn_arm_l_capsule, center=<5.3842 0 2.9557>, rotation=<rot -24.4787 0 0>}
 ]
 dyn_arm_r [
  {shapeType=Capsule, length=3.6134, radius=2, shapeName=dyn_arm_r_capsule, center=<4.3234 0 0>, rotation=<rot 90 0 0>}
  {shapeType=Capsule, length=3.1798, radius=1.4409, shapeName=dyn_arm_r_capsule, center=<5.3842 0 2.9557>, rotation=<rot -24.4787 0 0>}
 ]
 dyn_bow_l [
  {shapeType=Capsule, length=1.3122, radius=1.378, shapeName=dyn_bow_l_capsule, center=<2.5 0 0>, rotation=<rot 90 0 0>}
  {shapeType=Capsule, length=4.3906, radius=1.828, shapeName=dyn_bow_l_capsule, center=<4.4136 0 0.3065>, rotation=<rot -16.9759 0 -6.5478>}
 ]
 dyn_bow_r [
  {shapeType=Capsule, length=1.3122, radius=1.378, shapeName=dyn_bow_r_capsule, center=<4 0 0>, rotation=<rot 90 0 0>}
  {shapeType=Capsule, length=4.3906, radius=1.828, shapeName=dyn_bow_r_capsule, center=<4.4136 0 0.3065>, rotation=<rot -16.9759 0 6.5478>}
 ]
 dyn_leg_l [{shapeType=Capsule, length=3.5437, radius=2.5, shapeName=dyn_leg_l_capsule, center=<4.1932 0 0>, rotation=<rot 90 0 0>}]
 dyn_leg_r [{shapeType=Capsule, length=3.5437, radius=2.5, shapeName=dyn_leg_r_capsule, center=<3.5 0 0>, rotation=<rot 90 0 0>}]
}
constraints [
 {bone1Name=dyn_arm_l, bone2Name=base_body, swing1LimitDegrees=60, swing2LimitDegrees=25, twistLimitDegrees=1}
 {bone1Name=dyn_arm_r, bone2Name=base_body, swing1LimitDegrees=60, swing2LimitDegrees=25, twistLimitDegrees=1}
 {bone1Name=dyn_bow_l, bone2Name=base_body, swing1LimitDegrees=20, swing2LimitDegrees=20, twistLimitDegrees=1}
 {bone1Name=dyn_bow_r, bone2Name=base_body, swing1LimitDegrees=20, swing2LimitDegrees=20, twistLimitDegrees=1}
 {bone1Name=dyn_leg_l, bone2Name=base_body, swing1LimitDegrees=45, swing2LimitDegrees=25, twistLimitDegrees=1}
 {bone1Name=dyn_leg_r, bone2Name=base_body, swing1LimitDegrees=45, swing2LimitDegrees=25, twistLimitDegrees=1}
]
